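# Physics asset for the M_MED_AccentWall backpack cosmetic, defining collision bodies and joints for its three tentacles and head bones.
# Tentacles A, B and C are capsule chains, each starting at a small MSTR sphere, with alternating small and large capsules along much of their length.
# A two-link headUp chain hangs off head, while pack and head_Main are unconstrained capsules.
# Swing limits vary per chain, loosest on A, and every joint has tight twist.
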PhysicsAsset /BRCosmetics/Accessories/FORT_Backpacks/Backpack_M_MED_AccentWall/Meshes/M_MED_AccentWall_Pack_Physics
bodies {
 dyn_TentacleA_1 [{shapeType=Capsule, length=0, radius=2.2793, shapeName=dyn_TentacleA_1_capsule, center=<1.5533 0 0>, rotation=<rot 90 0 0>}]
 dyn_TentacleA_2 [{shapeType=Capsule, length=1, radius=3.6, shapeName=dyn_TentacleA_2_capsule, center=<1.1865 0 0.2361>, rotation=<rot 90 0 0>}]
 dyn_TentacleA_3 [{shapeType=Capsule, length=0.0127, radius=1.9873, shapeName=dyn_TentacleA_3_capsule, center=<2.0613 0 0>, rotation=<rot 90 0 0>}]
 dyn_TentacleA_4 [{
   shapeType=Capsule
   length=0.0973
   radius=4.75
   shapeName=dyn_TentacleA_4_capsule
   center=<3.3918 0 1.4747>
   rotation=<rot 84.4531 179.9998 179.9999>
  }]
 dyn_TentacleA_5 [{shapeType=Capsule, length=0, radius=2.0775, shapeName=dyn_TentacleA_5_capsule, center=<1.5533 0 0>, rotation=<rot 90 0 0>}]
 dyn_TentacleA_6 [{shapeType=Capsule, length=0.1916, radius=2.0076, shapeName=dyn_TentacleA_6_capsule, center=<1.5533 0 0>, rotation=<rot 90 0 0>}]
 dyn_TentacleA_7 [{shapeType=Capsule, length=1, radius=4.986, shapeName=dyn_TentacleA_7_capsule, center=<4.3189 0 1.6058>, rotation=<rot 90 0 0>}]
 dyn_TentacleA_8 [{shapeType=Capsule, length=0, radius=2.0728, shapeName=dyn_TentacleA_8_capsule, center=<1.5533 0 0>, rotation=<rot 90 0 0>}]
 dyn_TentacleA_MSTR [{shapeType=Sphere, radius=1, shapeName=dyn_TentacleA_MSTR_sphere}]
 dyn_TentacleB_1 [{shapeType=Capsule, length=1, radius=1, shapeName=dyn_TentacleB_1_capsule, center=<1.3192 0 0>, rotation=<rot 90 0 0>}]
 dyn_TentacleB_2 [{shapeType=Capsule, length=1.5696, radius=3.4897, shapeName=dyn_TentacleB_2_capsule, center=<0.7371 0 0.9128>, rotation=<rot 77.4717 0 0>}]
 dyn_TentacleB_3 [{shapeType=Capsule, length=1, radius=1, shapeName=dyn_TentacleB_3_capsule, center=<1.3192 0 0>, rotation=<rot 90 0 0>}]
 dyn_TentacleB_4 [{shapeType=Capsule, length=2.0164, radius=3.3024, shapeName=dyn_TentacleB_4_capsule, center=<1.3192 -0.0452 1.3554>, rotation=<rot 90 0 0>}]
 dyn_TentacleB_5 [{shapeType=Capsule, length=1, radius=1, shapeName=dyn_TentacleB_5_capsule, center=<1.5129 0 0>, rotation=<rot 90 0 0>}]
 dyn_TentacleB_6 [{shapeType=Capsule, length=1.4271, radius=3.9108, shapeName=dyn_TentacleB_6_capsule, center=<3.5045 0 0>, rotation=<rot 90 0 0>}]
 dyn_TentacleB_7 [{shapeType=Capsule, length=1, radius=1, shapeName=dyn_TentacleB_7_capsule, center=<1.3192 0 0>, rotation=<rot 90 0 0>}]
 dyn_TentacleB_MSTR [{shapeType=Sphere, radius=1, shapeName=dyn_TentacleB_MSTR_sphere}]
 dyn_TentacleC_1 [{shapeType=Capsule, length=0.5618, radius=1.6443, shapeName=dyn_TentacleC_1_capsule, center=<1.5334 0 0>, rotation=<rot 90 0 0>}]
 dyn_TentacleC_2 [{shapeType=Capsule, length=1.2601, radius=3.7646, shapeName=dyn_TentacleC_2_capsule, center=<1.9334 0 0.8339>, rotation=<rot 90 0 0>}]
 dyn_TentacleC_3 [{shapeType=Capsule, length=0.1906, radius=1.8094, shapeName=dyn_TentacleC_3_capsule, center=<2.0217 0 0>, rotation=<rot 90 0 0>}]
 dyn_TentacleC_4 [{shapeType=Capsule, length=0.103, radius=1.897, shapeName=dyn_TentacleC_4_capsule, center=<2.0276 0 0>, rotation=<rot 90 0 0>}]
 dyn_TentacleC_5 [{shapeType=Capsule, length=0.8117, radius=4.577, shapeName=dyn_TentacleC_5_capsule, center=<2.4987 0 1.033>, rotation=<rot 90 0 0>}]
 dyn_TentacleC_6 [{shapeType=Capsule, length=1, radius=1.8962, shapeName=dyn_TentacleC_6_capsule, center=<2.5309 0 0>, rotation=<rot 90 0 0>}]
 dyn_TentacleC_7 [{shapeType=Capsule, length=0, radius=5.9261, shapeName=dyn_TentacleC_7_capsule, center=<4.9511 0 0.3335>, rotation=<rot 90 0 0>}]
 dyn_TentacleC_8 [{shapeType=Capsule, length=0.6618, radius=1.7645, shapeName=dyn_TentacleC_8_capsule, center=<2.0276 0 0>, rotation=<rot 90 0 0>}]
 dyn_TentacleC_MSTR [{shapeType=Sphere, radius=1, shapeName=dyn_TentacleC_MSTR_sphere}]
 head [{shapeType=Capsule, length=2.1598, radius=5.7046, shapeName=head_capsule, center=<1.1557 0.1196 0.1273>, rotation=<rot -90 180 -176.6862>}]
 headUp_1 [{shapeType=Capsule, length=2, radius=2.5, shapeName=headUp_1_capsule, center=<3.2426 -0.0486 0.2956>, rotation=<rot -90 180 180>}]
 headUp_2 [{shapeType=Capsule, length=1.5, radius=4.5, shapeName=headUp_2_capsule, center=<6.6085 0.1532 -1.2033>, rotation=<rot -75.2523 180 -179.9999>}]
 head_Main [{shapeType=Capsule, length=8.134, radius=7.88, shapeName=head_Main_capsule, center=<-11.726 0.3183 0.21>, rotation=<rot -90 180 176.9005>}]
 pack [{shapeType=Capsule, length=26.5, radius=44, shapeName=pack_capsule, center=<4.3687 40.7993 -2.8096>, rotation=<rot 0 0 -10.348>}]
}
constraints [
 {bone1Name=headUp_1, bone2Name=head, swing1LimitDegrees=40, swing2LimitDegrees=40, twistLimitDegrees=5}
 {bone1Name=headUp_2, bone2Name=headUp_1, swing1LimitDegrees=40, swing2LimitDegrees=40, twistLimitDegrees=5}
 {bone1Name=dyn_TentacleC_MSTR, bone2Name=dyn_TentacleC_1, swing1LimitDegrees=50, swing2LimitDegrees=50, twistLimitDegrees=5}
 {bone1Name=dyn_TentacleC_1, bone2Name=dyn_TentacleC_2, swing1LimitDegrees=50, swing2LimitDegrees=50, twistLimitDegrees=5}
 {bone1Name=dyn_TentacleC_2, bone2Name=dyn_TentacleC_3, swing1LimitDegrees=50, swing2LimitDegrees=50, twistLimitDegrees=5}
 {bone1Name=dyn_TentacleC_3, bone2Name=dyn_TentacleC_4, swing1LimitDegrees=50, swing2LimitDegrees=50, twistLimitDegrees=5}
 {bone1Name=dyn_TentacleC_4, bone2Name=dyn_TentacleC_5, swing1LimitDegrees=50, swing2LimitDegrees=50, twistLimitDegrees=5}
 {bone1Name=dyn_TentacleC_5, bone2Name=dyn_TentacleC_6, swing1LimitDegrees=50, swing2LimitDegrees=50, twistLimitDegrees=5}
 {bone1Name=dyn_TentacleC_6, bone2Name=dyn_TentacleC_7, swing1LimitDegrees=50, swing2LimitDegrees=50, twistLimitDegrees=5}
 {bone1Name=dyn_TentacleC_7, bone2Name=dyn_TentacleC_8, swing1LimitDegrees=50, swing2LimitDegrees=50, twistLimitDegrees=5}
 {bone1Name=dyn_TentacleA_MSTR, bone2Name=dyn_TentacleA_1, swing1LimitDegrees=70, swing2LimitDegrees=75, twistLimitDegrees=5}
 {bone1Name=dyn_TentacleA_1, bone2Name=dyn_TentacleA_2, swing1LimitDegrees=70, swing2LimitDegrees=75, twistLimitDegrees=5}
 {bone1Name=dyn_TentacleA_2, bone2Name=dyn_TentacleA_3, swing1LimitDegrees=70, swing2LimitDegrees=75, twistLimitDegrees=5}
 {bone1Name=dyn_TentacleA_3, bone2Name=dyn_TentacleA_4, swing1LimitDegrees=70, swing2LimitDegrees=75, twistLimitDegrees=5}
 {bone1Name=dyn_TentacleA_4, bone2Name=dyn_TentacleA_5, swing1LimitDegrees=70, swing2LimitDegrees=75, twistLimitDegrees=5}
 {bone1Name=dyn_TentacleA_5, bone2Name=dyn_TentacleA_6, swing1LimitDegrees=70, swing2LimitDegrees=75, twistLimitDegrees=5}
 {bone1Name=dyn_TentacleA_6, bone2Name=dyn_TentacleA_7, swing1LimitDegrees=70, swing2LimitDegrees=75, twistLimitDegrees=5}
 {bone1Name=dyn_TentacleA_7, bone2Name=dyn_TentacleA_8, swing1LimitDegrees=70, swing2LimitDegrees=75, twistLimitDegrees=5}
 {bone1Name=dyn_TentacleB_MSTR, bone2Name=dyn_TentacleB_1, swing1LimitDegrees=45, swing2LimitDegrees=45, twistLimitDegrees=5}
 {bone1Name=dyn_TentacleB_1, bone2Name=dyn_TentacleB_2, swing1LimitDegrees=45, swing2LimitDegrees=45, twistLimitDegrees=5}
 {bone1Name=dyn_TentacleB_3, bone2Name=dyn_TentacleB_4, swing1LimitDegrees=45, swing2LimitDegrees=45, twistLimitDegrees=5}
 {bone1Name=dyn_TentacleB_4, bone2Name=dyn_TentacleB_5, swing1LimitDegrees=45, swing2LimitDegrees=45, twistLimitDegrees=5}
 {bone1Name=dyn_TentacleB_5, bone2Name=dyn_TentacleB_6, swing1LimitDegrees=45, swing2LimitDegrees=45, twistLimitDegrees=5}
 {bone1Name=dyn_TentacleB_6, bone2Name=dyn_TentacleB_7, swing1LimitDegrees=45, swing2LimitDegrees=45, twistLimitDegrees=5}
 {bone1Name=dyn_TentacleB_2, bone2Name=dyn_TentacleB_3, swing1LimitDegrees=45, swing2LimitDegrees=45, twistLimitDegrees=5}
]
