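# Physics asset for the AccentWall backpack cosmetic, named as its skydive variant, defining bodies and joints for three tentacles and a head chain.
# Tentacles A, B and C are capsule chains from small MSTR root spheres, linked by constraints with moderate swing and tight twist limits.
# The head and two headUp capsules form a much stiffer chain.
# One large, offset pack capsule has no constraints, nor do any link the chains together.
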PhysicsAsset /BRCosmetics/Accessories/FORT_Backpacks/Backpack_M_MED_AccentWall/Meshes/M_MED_AccentWall_Pack_Physics_SkyDive
bodies {
 dyn_TentacleA_1 [{
   shapeType=Capsule
   length=1.8372
   radius=2.6659
   shapeName=dyn_TentacleA_1_capsule
   center=<3.3567 -0.0001 -0.4287>
   rotation=<rot 56.8318 0.0128 0.0101>
  }]
 dyn_TentacleA_3 [{
   shapeType=Capsule
   length=2.7587
   radius=2.8405
   shapeName=dyn_TentacleA_3_capsule
   center=<4.7158 -0.1517 -0.2623>
   rotation=<rot 79.4399 0 0.0001>
  }]
 dyn_TentacleA_5 [{shapeType=Capsule, length=1.7043, radius=3.081, shapeName=dyn_TentacleA_5_capsule, center=<3.1225 0 0>, rotation=<rot 90 90.77 105.6307>}]
 dyn_TentacleA_7 [{shapeType=Capsule, length=0, radius=5.2976, shapeName=dyn_TentacleA_7_capsule, center=<3.5705 0 0>, rotation=<rot 90 0 0>}]
 dyn_TentacleA_MSTR [{shapeType=Sphere, radius=1, shapeName=dyn_TentacleA_MSTR_sphere}]
 dyn_TentacleB_1 [{
   shapeType=Capsule
   length=1.0699
   radius=1.8893
   shapeName=dyn_TentacleB_1_capsule
   center=<2.3785 -0.242 0.0414>
   rotation=<rot 73.5974 -0.0209 14.5327>
  }]
 dyn_TentacleB_3 [{shapeType=Capsule, length=0.1789, radius=2.6118, shapeName=dyn_TentacleB_3_capsule, center=<2.9541 0 -0.1421>, rotation=<rot 82.2006 0 0>}]
 dyn_TentacleB_5 [{
   shapeType=Capsule
   length=3.1721
   radius=3.223
   shapeName=dyn_TentacleB_5_capsule
   center=<3.7932 0 0.8659>
   rotation=<rot -74.0523 0.0001 0>
  }]
 dyn_TentacleB_MSTR [{shapeType=Sphere, radius=1, shapeName=dyn_TentacleB_MSTR_sphere}]
 dyn_TentacleC_1 [{
   shapeType=Capsule
   length=1.3575
   radius=2.3993
   shapeName=dyn_TentacleC_1_capsule
   center=<2.3846 0 0>
   rotation=<rot 73.8351 -0.0001 -0.0001>
  }]
 dyn_TentacleC_3 [{shapeType=Capsule, length=2.3659, radius=2.8943, shapeName=dyn_TentacleC_3_capsule, center=<3.1475 0 0>, rotation=<rot 90 0 0>}]
 dyn_TentacleC_5 [{shapeType=Capsule, length=2.5657, radius=3.6229, shapeName=dyn_TentacleC_5_capsule, center=<4.2933 0.0957 0>, rotation=<rot 90 180 174.8543>}]
 dyn_TentacleC_7 [{
   shapeType=Capsule
   length=2.6048
   radius=5.3973
   shapeName=dyn_TentacleC_7_capsule
   center=<4.9689 -0.3051 0>
   rotation=<rot 90 180 -178.4785>
  }]
 dyn_TentacleC_MSTR [{shapeType=Sphere, radius=1, shapeName=dyn_TentacleC_MSTR_sphere}]
 head [{
   shapeType=Capsule
   length=8.0289
   radius=7.8643
   shapeName=head_capsule
   center=<-11.6668 0.9634 0.3299>
   rotation=<rot 90 179.1048 -170.8952>
  }]
 headUp_1 [{shapeType=Capsule, length=2, radius=2.5, shapeName=headUp_1_capsule, center=<3.2426 -0.0486 0.2956>, rotation=<rot -90 180 180>}]
 headUp_2 [{shapeType=Capsule, length=3.5, radius=2.5, shapeName=headUp_2_capsule, center=<5.2053 0.1532 -0.834>, rotation=<rot -75.2523 180 -179.9999>}]
 pack [{shapeType=Capsule, length=26.5, radius=44, shapeName=pack_capsule, center=<4.3687 40.7993 -2.8096>, rotation=<rot 0 0 -10.348>}]
}
constraints [
 {bone1Name=headUp_1, bone2Name=head, swing1LimitDegrees=5, swing2LimitDegrees=5, twistLimitDegrees=5}
 {bone1Name=headUp_2, bone2Name=headUp_1, swing1LimitDegrees=5, swing2LimitDegrees=5, twistLimitDegrees=5}
 {bone1Name=dyn_TentacleA_MSTR, bone2Name=dyn_TentacleA_1, swing1LimitDegrees=35, swing2LimitDegrees=35, twistLimitDegrees=5}
 {bone1Name=dyn_TentacleA_1, bone2Name=dyn_TentacleA_3, swing1LimitDegrees=35, swing2LimitDegrees=35, twistLimitDegrees=5}
 {bone1Name=dyn_TentacleA_3, bone2Name=dyn_TentacleA_5, swing1LimitDegrees=35, swing2LimitDegrees=35, twistLimitDegrees=5}
 {bone1Name=dyn_TentacleA_5, bone2Name=dyn_TentacleA_7, swing1LimitDegrees=35, swing2LimitDegrees=35, twistLimitDegrees=5}
 {bone1Name=dyn_TentacleB_MSTR, bone2Name=dyn_TentacleB_1, swing1LimitDegrees=35, swing2LimitDegrees=35, twistLimitDegrees=5}
 {bone1Name=dyn_TentacleB_1, bone2Name=dyn_TentacleB_3, swing1LimitDegrees=35, swing2LimitDegrees=35, twistLimitDegrees=5}
 {bone1Name=dyn_TentacleB_3, bone2Name=dyn_TentacleB_5, swing1LimitDegrees=35, swing2LimitDegrees=35, twistLimitDegrees=5}
 {bone1Name=dyn_TentacleC_MSTR, bone2Name=dyn_TentacleC_1, swing1LimitDegrees=35, swing2LimitDegrees=35, twistLimitDegrees=5}
 {bone1Name=dyn_TentacleC_1, bone2Name=dyn_TentacleC_3, swing1LimitDegrees=35, swing2LimitDegrees=35, twistLimitDegrees=5}
 {bone1Name=dyn_TentacleC_3, bone2Name=dyn_TentacleC_5, swing1LimitDegrees=35, swing2LimitDegrees=35, twistLimitDegrees=5}
 {bone1Name=dyn_TentacleC_5, bone2Name=dyn_TentacleC_7, swing1LimitDegrees=35, swing2LimitDegrees=35, twistLimitDegrees=5}
]
