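# Animation Blueprint for the AccentWall backpack cosmetic, an octopus-like pack with wavy tentacles that panics when its wearer takes damage.
# Owner spine and head rotations offset the pack and head bones, while separate ground and skydive RigidBody simulations feed a cached pose.
# Idle, SkyDiving and Reactive states layer Control Rig tentacle waves with additive clips.
# Health or shield loss and front-end delays trigger Reactive, with notifies driving smoothed panic intensity.
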AnimBlueprint /BRCosmetics/Accessories/FORT_Backpacks/Backpack_M_MED_AccentWall/Meshes/M_MED_AccentWall_Pack_AnimBP
target_skeleton=&/M_MED_AccentWall_Pack_Skeleton
graphs [
 AnimGraph {nodes [
   AnimGraphNode_StateMachine_0 = Idle() {is=title AnimGraphNode_StateMachine}
   n1 = Local To Component(LocalPose=@AnimGraphNode_StateMachine_0)
   n2 = Get backpack_OctupusHeadRotate() -> (backpack_OctupusHeadRotate:Rotator)
   n3 = "Transform (Modify) Bone - Bone: pack_down"(
    ComponentPose="Transform (Modify) Bone - Bone: headUp_2"(
     ComponentPose="Transform (Modify) Bone - Bone: headUp_1"(ComponentPose=@n1, Alpha=1, Translation="0,0,0", Rotation=@n2, Scale="1,1,1") {rotationMode=BMM_Additive, alphaBoolBlend {blendOption=Linear}, boneToModify {boneName=headUp_1}}
     Alpha=1
     Translation="0,0,0"
     Rotation=Get backpack_OctupusHeadRotate()
     Scale="1,1,1"
    ) {rotationMode=BMM_Additive, alphaBoolBlend {blendOption=Linear}, boneToModify {boneName=headUp_2}}
    Alpha=1
    Translation="0,-2,0"
    Rotation="0,10,5"
    Scale="1,1,1"
   ) {
    rotationMode=BMM_Additive
    translationMode=BMM_Additive
    translationSpace=BCS_BoneSpace
    alphaBoolBlend {blendOption=Linear}
    boneToModify {boneName=pack_down}
    rotation {roll=5, yaw=10}
    translation {y=-2}
   }
   Output Pose(Result=Component To Local(ComponentPose=@n3)) {layerGroup=None}
   n4 = "Transform (Modify) Bone - Bone: pack"(
    ComponentPose=Local To Component(LocalPose=Input Pose())
    Translation=Get backpack_upTranslation() -> (backpack_upTranslation:Vector)
    Rotation=Get backpack_upRotate() -> (backpack_upRotate:Rotator)
   ) {rotationMode=BMM_Additive, translationMode=BMM_Additive, alphaBoolBlend {blendOption=Linear}, boneToModify {boneName=pack}}
   Save cached pose 'Input'(Pose=Component To Local(ComponentPose=@n4)) {cacheName=Input}
   n5 = RigidBody!(
    ComponentPose=Local To Component(LocalPose=Use cached pose 'Input'())
    Alpha=0.7
    OverrideWorldGravity:Vector=Get Gravity Override_Skydiving() -> (Gravity Override_Skydiving:Vector)
    ExternalForce="0,0,300"
   ) -> (Pose) {
    bOverrideWorldGravity=true
    overridePhysicsAsset=&/M_MED_AccentWall_Pack_Physics_SkyDive
    simulationSpace=BaseBoneSpace
    alphaBoolBlend {blendOption=Linear}
    baseBoneRef {boneName=root}
    componentAppliedLinearAccClamp=<700 700 700>
    componentLinearAccScale=<0.1 0.1 0.1>
    componentLinearVelScale=<0.1 0.1 0.1>
    overrideWorldGravity {z=-980}
   }
   n6 = RigidBody(
    ComponentPose=Local To Component(LocalPose=Use cached pose 'Input'())
    Alpha=0.5
    OverrideWorldGravity=Get Gravity Override()
    ExternalForce=Get extraForce_ground() -> (extraForce_ground:Vector)
   ) {
    bOverrideWorldGravity=true
    overridePhysicsAsset=&/M_MED_AccentWall_Pack_Physics
    simulationSpace=BaseBoneSpace
    alphaBoolBlend {blendOption=Linear}
    baseBoneRef {boneName=root}
    componentAppliedLinearAccClamp=<700 700 700>
    componentLinearAccScale=<1 1 1>
    componentLinearVelScale=<0.4 0.3 0.3>
    overrideWorldGravity {z=-980}
   }
   n7 = Layered blend per bone(BasePose=Component To Local(ComponentPose=@n6), BlendPoses_0=Use cached pose 'Input'(), BlendWeights_0=0.7) {layerSetup [{branchFilters [{blendDepth=5, boneName=dyn_TentacleA_4}, {blendDepth=4, boneName=dyn_TentacleB_4}, {blendDepth=5, boneName=dyn_TentacleC_4}]}]}
   Save cached pose 'Dynamic_Cache'(Pose=Blend Poses by bool(BlendPose_0=Component To Local(ComponentPose=@n5), BlendPose_1=@n7, BlendTime_0=0.5, BlendTime_1=0.5, bActiveValue=Get bIsSkyDiving()) {blendType=Linear}) {cacheName=Dynamic_Cache}
  ]}
 Idle {
  entry=Idle
  path=AnimGraph.AnimGraphNode_StateMachine_0.Idle
  ids=id
  nodes [
   AnimStateNode_0 = Idle() {is=title AnimStateNode}
   AnimStateNode_1 = SkyDiving() {is=title AnimStateNode}
   AnimStateNode_2 = Reactive() {is=title AnimStateNode, stateEntered {guid=#1, notifyName=EnteredReactive}}
   AnimStateTransitionNode_0 = AnimStateTransitionNode() {crossfadeDuration=0.5}
   AnimStateTransitionNode_2 = AnimStateTransitionNode() {crossfadeDuration=0.5}
   AnimStateTransitionNode_3 = AnimStateTransitionNode() {crossfadeDuration=0.5}
   AnimStateTransitionNode_4 = AnimStateTransitionNode() {crossfadeDuration=0.5}
   AnimStateTransitionNode_5 = AnimStateTransitionNode() {crossfadeDuration=0.35}
   AnimStateTransitionNode_6 = AnimStateTransitionNode() {crossfadeDuration=0.5}
  ]
  links [
   {dst=Reactive, rule=AnimGraph.AnimGraphNode_StateMachine_0.Idle.AnimStateTransitionNode_6.Transition, src=Idle, via=AnimStateTransitionNode_6}
   {
    dst=SkyDiving
    rule=AnimGraph.AnimGraphNode_StateMachine_0.Idle.AnimStateTransitionNode_0.Transition
    src=Idle
    via=AnimStateTransitionNode_0
   }
   {dst=Idle, rule=AnimGraph.AnimGraphNode_StateMachine_0.Idle.AnimStateTransitionNode_3.Transition, src=Reactive, via=AnimStateTransitionNode_3}
   {
    dst=SkyDiving
    rule=AnimGraph.AnimGraphNode_StateMachine_0.Idle.AnimStateTransitionNode_5.Transition
    src=Reactive
    via=AnimStateTransitionNode_5
   }
   {
    dst=Idle
    rule=AnimGraph.AnimGraphNode_StateMachine_0.Idle.AnimStateTransitionNode_4.Transition
    src=SkyDiving
    via=AnimStateTransitionNode_4
   }
   {
    dst=Reactive
    rule=AnimGraph.AnimGraphNode_StateMachine_0.Idle.AnimStateTransitionNode_2.Transition
    src=SkyDiving
    via=AnimStateTransitionNode_2
   }
  ]
 }
 Idle {
  path=AnimGraph.AnimGraphNode_StateMachine_0.Idle.AnimStateNode_0.Idle
  nodes [
   n1 = Control Rig!(
    Source=Use cached pose 'Dynamic_Cache'()
    Alpha
    scale_Wave_Speed:double=Get tentacles_IdleSpeedScale() -> (tentacles_IdleSpeedScale:double)
    scale_Wave_Amplitude:double=Get tentacles_IdleAmplitudeScale() -> (tentacles_IdleAmplitudeScale:double)
   ) -> (Pose) {
    alphaBoolBlend {blendOption=Linear}
    controlRigAssetReference {blueprintRigClass=&/M_MED_AccentWall_Pack_CtrlRig.M_MED_AccentWall_Pack_CtrlRig_C}
   }
   Output Animation Pose(Result=Apply Additive(
     Base=@n1
     Additive="Sequence Player (additive) 'AccentWall_Idle_PACK'"() {sequence=/BRCosmetics/Accessories/FORT_Backpacks/Backpack_M_MED_AccentWall/Animation/AccentWall_Idle_PACK}
    ) {alphaBoolBlend {blendOption=Linear}})
  ]
 }
 SkyDiving {
  path=AnimGraph.AnimGraphNode_StateMachine_0.Idle.AnimStateNode_1.SkyDiving
  nodes [
   n1 = Control Rig(
    Source=Use cached pose 'Dynamic_Cache'()
    scale_Wave_Speed=Get tentacles_PanicSpeedScale() -> (tentacles_PanicSpeedScale:double)
    scale_Wave_Amplitude=Get tentacles_PanicAmplitudeScale() -> (tentacles_PanicAmplitudeScale:double)
   ) {
    alphaBoolBlend {blendOption=Linear}
    controlRigAssetReference {blueprintRigClass=&M_MED_AccentWall_Pack_CtrlRig_C}
   }
   n2 = "Sequence Player (additive) 'AccentWall_Skydiving_PACK'"() {sequence=/BRCosmetics/Accessories/FORT_Backpacks/Backpack_M_MED_AccentWall/Animation/AccentWall_Skydiving_PACK}
   n3 = "Transform (Modify) Bone - Bone: pack_down"(
    ComponentPose=Local To Component(LocalPose=Apply Additive(Base=@n1, Additive=@n2) {alphaBoolBlend {blendOption=Linear}})
    Alpha=1
    Translation="0,0,0"
    Rotation=Get backpack_downRotate() -> (backpack_downRotate:Rotator)
    Scale="1,1,1"
   ) {rotationMode=BMM_Additive, translationMode=BMM_Additive, alphaBoolBlend {blendOption=Linear}, boneToModify {boneName=pack_down}}
   Output Animation Pose(Result=Component To Local(ComponentPose=@n3))
  ]
 }
 Reactive {
  path=AnimGraph.AnimGraphNode_StateMachine_0.Idle.AnimStateNode_2.Reactive
  nodes [
   n1 = Control Rig!(
    Source=Use cached pose 'Dynamic_Cache'()
    Alpha
    scale_Wave_Speed=Get tentacles_PanicSpeedScale_Smooth() -> (tentacles_PanicSpeedScale_Smooth:double)
    scale_Wave_Amplitude=Get tentacles_PanicAmplitudeScale_Smooth() -> (tentacles_PanicAmplitudeScale_Smooth:double)
    panicWaveSpeed:double=5
   ) -> (Pose) {
    alphaBoolBlend {blendOption=Linear}
    controlRigAssetReference {blueprintRigClass=&M_MED_AccentWall_Pack_CtrlRig_C}
   }
   Output Animation Pose(Result=Apply Additive(
     Base=@n1
     Additive="Sequence Player (additive) 'AccentWall_Damage_PACK'"() {bLoopAnimation=false, sequence=/BRCosmetics/Accessories/FORT_Backpacks/Backpack_M_MED_AccentWall/Animation/AccentWall_Damage_PACK}
    ) {alphaBoolBlend {blendOption=Linear}})
  ]
 }
 Transition {
  path=AnimGraph.AnimGraphNode_StateMachine_0.Idle.AnimStateTransitionNode_0.Transition
  nodes [Result(bCanEnterTransition=Get bIsSkydiveFloating())]
 }
 Transition {
  path=AnimGraph.AnimGraphNode_StateMachine_0.Idle.AnimStateTransitionNode_2.Transition
  nodes [Result(bCanEnterTransition=Get ReactiveAnimation() -> (ReactiveAnimation:bool))]
 }
 Transition {
  path=AnimGraph.AnimGraphNode_StateMachine_0.Idle.AnimStateTransitionNode_3.Transition
  nodes [
   n1 = Get bIsSkyDiving()
   n2 = "Get Relevant Anim Time Remaining (Reactive)"!() -> (ReturnValue:float) {is=title K2Node_AnimGetter}
   Result(bCanEnterTransition=AND Boolean(A=NOT Boolean(A=@n1), B="float <= float"(A=@n2, B=0.2) -> (ReturnValue=~)))
  ]
 }
 Transition {
  path=AnimGraph.AnimGraphNode_StateMachine_0.Idle.AnimStateTransitionNode_4.Transition
  nodes [
   n1 = Get bIsSkyDiving()
   Result(bCanEnterTransition=XOR Boolean!(A:bool=Get bIsParachuteOpened(), B:bool=NOT Boolean(A=@n1)) -> (ReturnValue:bool))
  ]
 }
 Transition {
  path=AnimGraph.AnimGraphNode_StateMachine_0.Idle.AnimStateTransitionNode_5.Transition
  nodes [
   n1 = "Get Relevant Anim Time Remaining (Reactive)"()
   Result(bCanEnterTransition=AND Boolean(A=Get bIsSkyDiving(), B="float <= float"(A=@n1, B=0.5) -> (ReturnValue=~)))
  ]
 }
 Transition {
  path=AnimGraph.AnimGraphNode_StateMachine_0.Idle.AnimStateTransitionNode_6.Transition
  nodes [Result(bCanEnterTransition=Get ReactiveAnimation())]
 }
 EventGraph {nodes [
   n1 = Event BlueprintUpdateAnimation()
   > Set Delta Time X(Delta Time X=@n1.DeltaTimeX)
   > Sequence()
   > n2 = Gravity Override State Function DEPRECATED(CharacterPartAnimInstance=Self-Reference(), GameTimeInSeconds=GetGameTimeInSeconds(), GravityOverrideParams=Get Gravity Override Params())
   > n3 = Set Gravity Override(Gravity Override=@n2)
   > Set Gravity Override_Skydiving(Gravity Override_Skydiving:Vector=vector * vector(A=@n3, B:Vector="0.4,0.4,0.4") -> (ReturnValue=~)) -> (Output_Get:Vector)
   n4 = GetRightVector!(self:Scene Component Object Reference=GetOwnerSkeletalMeshComponent()) -> (ReturnValue:Vector)
   n5 = Get bIsFalling()
   n6 = Get PawnVelocity()
   > Set extraForce_ground(extraForce_ground:Vector=vector + vector(
     A=vector * vector(A=@n6, B="-2.5,-2.5,-2.5") -> (ReturnValue=~)
     B=vector * vector!(A=@n4, B="4000,4000,6000", C_X:bool=@n5, C_Y:bool=@n5, C_Z:bool=@n5) -> (ReturnValue=~)
    )) -> (Output_Get:Vector)
   n7 = GetDeltaTransformFromRefPose!(self=GetOwnerSkeletalMeshComponent(), BoneName=spine_05, BaseName=spine_01) -> (
    ReturnValue_Location_X:double
    ReturnValue_Location_Y:double
    ReturnValue_Location_Z:double
    ReturnValue_Rotation:Rotator
    ReturnValue_Scale:Vector
   )
   n8 = MapRangeClamped(Value=@n7, InRangeA=1, InRangeB=12, OutRangeA=75, OutRangeB=0)
   n9 = BreakRotator(InRot=@n7.ReturnValue_Rotation)
   n10 = BreakRotator(InRot=GetDeltaTransformFromRefPose(self=GetOwnerSkeletalMeshComponent(), BoneName=spine_05, BaseName=spine_04).ReturnValue_Rotation)
   n11 = vector + vector!(
    A=Get backpack_BeginOffset_Rotation() -> (backpack_BeginOffset_Rotation:Vector)
    B_X:double=float + float(
     A=MapRangeClamped(Value=@n9.Yaw, InRangeA=0, InRangeB=-90, OutRangeB=@n8)
     B=MapRangeClamped(Value=@n10.Yaw, InRangeA=0, InRangeB=-20, OutRangeB=20)
    )
    B_Y:double=0
    B_Z:double=0
   ) -> (ReturnValue_X:double, ReturnValue_Y:double, ReturnValue_Z:double)
   > Set backpack_upRotate(backpack_upRotate:Rotator=MakeRotator(Roll=@n11, Pitch=@n11.ReturnValue_Y, Yaw=@n11.ReturnValue_Z)) -> (Output_Get:Rotator)
   n12 = MapRangeClamped(Value=@n9.Yaw, InRangeA=-20, InRangeB=-90, OutRangeB=-5)
   n13 = BreakRotator(InRot=GetDeltaTransformFromRefPose(self=GetOwnerSkeletalMeshComponent(), BoneName=head, BaseName=spine_02).ReturnValue_Rotation)
   n14 = MapRangeClamped(Value=@n13.Yaw, InRangeA=-80, InRangeB=-150, OutRangeB=-7)
   n15 = MapRangeClamped(Value=@n9.Yaw, InRangeA=0, InRangeB=-90, OutRangeB=-12)
   n16 = MapRangeClamped(Value=@n13.Yaw, InRangeA=-80, InRangeB=-150, OutRangeB=7)
   > Set backpack_upTranslation(backpack_upTranslation:Vector=vector + vector!(
     A_X:double=0
     A_Y:double=float + float(A=@n12, B=@n14)
     A_Z:double=float + float(A=@n15, B=@n16)
     B:Vector=Get backpack_BeginOffset_Translation() -> (backpack_BeginOffset_Translation:Vector)
    ) -> (ReturnValue:Vector)) -> (Output_Get:Vector)
   n17 = MapRangeClamped(Value=@n13.Yaw, InRangeA=-10, InRangeB=-100, OutRangeB=-40)
   > Set backpack_OctupusHeadRotate(backpack_OctupusHeadRotate:Rotator=MakeRotator(Roll=@n17)) -> (Output_Get:Rotator)
   > Set backpack_downRotate(backpack_downRotate:Rotator=MakeRotator(Roll=float / float(A=@n17, B=2.5))) -> (Output_Get:Rotator)
   > Set tentacles_PanicSpeedScale_Smooth(tentacles_PanicSpeedScale_Smooth:double=FInterpTo(Current=Get tentacles_PanicSpeedScale_Smooth(), Target=Get tentacles_PanicSpeedScale(), DeltaTime=Get Delta Time X(), InterpSpeed=2)) -> (Output_Get:double)
   > Set tentacles_PanicAmplitudeScale_Smooth(tentacles_PanicAmplitudeScale_Smooth:double=FInterpTo(Current=Get tentacles_PanicAmplitudeScale_Smooth(), Target=Get tentacles_PanicAmplitudeScale(), DeltaTime=Get Delta Time X(), InterpSpeed=2)) -> (Output_Get:double)
   n18 = Event BlueprintProcessAnimInputEvent()
   > Branch(Condition=OR Boolean(
     A=IsNamed(self=@n18.AnimInputEvent, InEventName=aie.PlayerVitality.ShieldsDecreased)
     B=IsNamed(self=@n18.AnimInputEvent, InEventName=aie.PlayerVitality.HealthDecreased)
    ))
   > n19 = Branch(Condition=Get ReactiveAnimation())
   Set ReactiveAnimation(execute=@n19.else, ReactiveAnimation:bool=true) -> (Output_Get:bool)
   > RetriggerableDelay(Duration=0.2)
   > Set ReactiveAnimation(ReactiveAnimation=false)
   Event BlueprintPostEvaluateAnimation() {is=title K2Node_Event}
   > Set ReactiveAnimation(ReactiveAnimation=false)
   AnimNotify_RampDownIntensity() {is=title K2Node_Event}
   > n20 = Sequence()
   > Set tentacles_PanicSpeedScale(tentacles_PanicSpeedScale:double=2) -> (Output_Get:double)
   > Set tentacles_PanicAmplitudeScale(tentacles_PanicAmplitudeScale:double=1) -> (Output_Get:double)
   n21 = Set ReactiveAnimation(execute=@n20.then_1, ReactiveAnimation=true)
   Branch(execute=[@n22.then, @n21.then], Condition=Get bIsInFrontEnd())
   > Delay(Duration=10)
   > Set ReactiveAnimation(ReactiveAnimation=true)
   n22 = Event BlueprintInitializeAnimation()
   AnimNotify_EnteredReactive() {is=title K2Node_Event}
   > Set tentacles_PanicSpeedScale(tentacles_PanicSpeedScale=8)
   > Set tentacles_PanicAmplitudeScale(tentacles_PanicAmplitudeScale=2)
  ]}
]
variables [
 backpack_BeginOffset_Rotation {type=Vector, default=<6 0 0>}
 backpack_BeginOffset_Translation {type=Vector, default=<0 4 0>}
 backpack_downRotate {type=Rotator}
 backpack_OctupusHeadRotate {type=Rotator}
 backpack_upRotate {type=Rotator}
 backpack_upTranslation {type=Vector, default=<0 6 0>}
 Delta Time X {type=number}
 extraForce_ground {type=Vector}
 Gravity Override {type=Vector, default=<0 0 -980>}
 Gravity Override Params {type=GravityOverrideParamsStruct}
 Gravity Override_Skydiving {type=Vector}
 Head Main Location Offset {type=Vector}
 Head Main Location Offset Input {type=number, default=20}
 Head Main Rotation Offset {type=Rotator}
 Head Main Rotation Offset Input {type=number, default=20}
 panicDamageDuration {type=number, default=4.7}
 ReactiveAnimation {type=boolean}
 tentacles_IdleAmplitudeScale {type=number, default=1}
 tentacles_IdleSpeedScale {type=number, default=2}
 tentacles_PanicAmplitudeScale {type=number, default=2}
 tentacles_PanicAmplitudeScale_Smooth {type=number, default=2}
 tentacles_PanicSpeedScale {type=number, default=8}
 tentacles_PanicSpeedScale_Smooth {type=number, default=8}
]
